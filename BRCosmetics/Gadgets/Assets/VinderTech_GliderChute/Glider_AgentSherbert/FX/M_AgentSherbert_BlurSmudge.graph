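# Translucent, two-sided material for Niagara mesh particles, named as a blur smudge effect for the Agent Sherbert glider.
# It multiplies a center-scaled, clamped mask by a distortion texture, each sampled with the G component of its UVs duplicated.
# Dynamic-parameter and time offsets enter only as X, so that G-only sampling appears to discard them.
# Particle color tints the mask, while particle alpha and a boosted U-edge falloff shape a second, opacity-like term.
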Material /BRCosmetics/Gadgets/Assets/VinderTech_GliderChute/Glider_AgentSherbert/FX/M_AgentSherbert_BlurSmudge
graph {
 properties {
  bUsedWithNiagaraMeshParticles=true
  blendMode=BLEND_Translucent
  pixelDepthOffsetMode=PDOM_Legacy
  twoSided=true
  usedShadingModels=Default Lit
 }
 nodes [
  n1 = TextureCoordinate()
  n2 = ComponentMask(None=VectorParameter() {parameterName=Scale, defaultValue=<rgba 2 1.2 0 1>}.RGB) {g=true, r=true}
  n3 = DynamicParameter() {paramNames [OffsetUVs,Param2,Param3,Param4]}
  n4 = MaterialFunctionCall(Float2=Add(
    A=MaterialFunctionCall(UVs=@n1, Texture Scale=@n2) {materialFunction=&ScaleUVsByCenter}.UVs
    B=MaterialFunctionCall(X=@n3.OffsetUVs) {materialFunction=&MakeFloat2}.Result
   )) {materialFunction=&BreakOutFloat2Components}
  n5 = OneMinus(None=Time())
  n6 = MaterialFunctionCall(Float2=Add(
    A=TextureCoordinate()
    B=MaterialFunctionCall(X=@n5) {materialFunction=&MakeFloat2}.Result
   )) {materialFunction=&BreakOutFloat2Components}
  n7 = Multiply(
   A=TextureSample(UVs=MaterialFunctionCall(X=@n4.G, Y=@n4.G) {materialFunction=&MakeFloat2}.Result) {samplerSource=SSM_Clamp_WorldGroupSettings, texture=/Game/WaterAndWind/Textures/WeaponFireVectorMask}.G
   B=TextureSample(UVs=MaterialFunctionCall(X=@n6.G, Y=@n6.G) {materialFunction=&MakeFloat2}.Result) {texture=/Game/Global/Textures/Default/Masks_and_Gradients/T_BlasterBeamDistort}.R
  )
  n8 = ComponentMask(None=TextureCoordinate()) {r=true}
  n9 = Power(Base=Multiply(A=@n8, B=OneMinus(None=@n8)), Exp=Constant() {r=2})
  n10 = Constant() {r=40}
  n11 = ParticleColor()
  Multiply(A=Multiply(A=@n7, B=Multiply(A=@n9, B=@n10) {constB=10}), B=@n11.A)
  Multiply(A=@n11.RGB, B=@n7)
  TextureCoordinate()
 ]
}
