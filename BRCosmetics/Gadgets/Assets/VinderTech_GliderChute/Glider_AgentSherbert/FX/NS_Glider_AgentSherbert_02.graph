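# Niagara effect for the Agent Sherbert glider, drawing black wings, trails, smoke and droplets on skeletal mesh sockets.
# WingsSplines and Wings_Smudge place wing meshes, with Smudge feeding lagged, angle-clamped endpoints to materials through dynamic parameters for secondary motion.
# Anim-trail ribbons, flipbook smoke and GPU drops add motion.
# System timelines and a front-end flag adjust fades and directions; two debug-style emitters are disabled.
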NiagaraSystem /BRCosmetics/Gadgets/Assets/VinderTech_GliderChute/Glider_AgentSherbert/FX/NS_Glider_AgentSherbert_02
dependencies {
 usedDataInterfaces [
  &NiagaraDataInterfacePlatformSet
  &NiagaraDataInterfaceCurve
  &NiagaraDataInterfaceSpriteRendererInfo
  &NiagaraDataInterfaceArrayFloat3
  &NiagaraDataInterfaceArrayFloat
  &NiagaraDataInterfaceMeshRendererInfo
  &NiagaraDataInterfaceColorCurve
  &NiagaraDataInterfaceVector2DCurve
 ]
 usedDynamicInputs [&SetBoolByFloatComparison, &Abs_Float]
 usedModules [
  &SystemState
  &Timeline
  &self:SystemSpawnScript.SystemScriptSource.SystemScriptGraph.NiagaraNodeAssignment_0.SetVariables_#1
  &self:SystemSpawnScript.SystemScriptSource.SystemScriptGraph.NiagaraNodeAssignment_1.SetVariables_#2
 ]
}
emitters {
 Drops {
  simTarget=GPUComputeSim
  emitterUpdateScript [
   &EmitterState(Inactive Response=::NewEnumerator1, Life Cycle Mode=::NewEnumerator1, Loop Behavior=::NewEnumerator1, Loop Duration=3600, UseLoopDelay=false)
   &SpawnRate(
    SpawnGroup:NiagaraInt32=&RandomRangeInt(Minimum=0, Maximum=1)
    SpawnRate=32
    Use Spawn Probability=false
   )
  ]
  particleSpawnScript [
   &FN_SetSpawnGroup(Number of Spawn Groups=2)
   &InitializeParticle(
    Color=&MakeCustomLinearColorFromBool(
     Bool=$System.DefaultVarient
     True Color=&MakeCustomLinearColorFromBool(Bool=&RandomBool(), True Color=<rgba 0 0 0 0.8>, False Color=<rgba 0 0 0 0.8>)
     False Color=&MakeCustomLinearColorFromBool(Bool=&RandomBool(), True Color=<rgba 0 0 0 0.8>, False Color=<rgba 0 0 0 0.8>)
    )
    Lifetime Max=0.8
    Lifetime Min=0.8
    Lifetime Mode=::NewEnumerator1
    Sprite Rotation Mode=::NewEnumerator2
    Sprite Size Mode=::NewEnumerator1
    UsePositionOffset=false
    Write Material Random=true
   ) {module=InitializeParticle001}
   &SkeletalMeshLocation(
    Apply To Particle Position=false
    Bone Sampling Mode=::NewEnumerator0
    Inherit Velocity=true
    "Inherit Velocity (Scale)"=0.1
    Inherited Velocity Speed Limit=1
    Interpolate Mesh Sampling=false
    MaskSpawnGroup=true
    Module.Override Randomness=false
    Sampled Position Offset Coordinate Space:ENiagaraMeshTransforms=::NewEnumerator2
    Sampled Velocity Coordinate Space:ENiagaraCoordinateSpace=::Local
    Skeletal Mesh=$User.SkeletalMeshRight
    Use Sampled Mesh Attribute as Ribbon ID=false
    VelocitySpeedLimit=true
   ) {enabled=false}
   &SkeletalMeshLocation(
    Apply To Particle Position=false
    Bone / Socket Index=&MaskIntBySpawnGroup(
     Masked Int=&RandomRangeInt(Minimum=0, Maximum=7)
     Passthrough Int=&RandomRangeInt(Minimum=8, Maximum=15)
     Spawn Group Mask=0
    )
    Inherit Velocity=true
    "Inherit Velocity (Scale)"=0.1
    Inherited Velocity Speed Limit=1
    Interpolate Mesh Sampling=false
    MaskSpawnGroup=true
    Mesh Sampling Type=::NewEnumerator5
    Sampled Position Offset Coordinate Space:ENiagaraMeshTransforms=::NewEnumerator2
    Sampled Velocity Coordinate Space:ENiagaraCoordinateSpace=::Local
    Skeletal Mesh=$User.SkeletalMeshLeft
    Socket Sampling Mode=::NewEnumerator5
    Spawn Group Mask=1
    Use Sampled Mesh Attribute as Ribbon ID=false
    VelocitySpeedLimit=true
   ) {module=SkeletalMeshLocation001}
   &AddVelocity(Velocity=<0 100 50>) {enabled=false}
   /SetVariables(Particles.Velocity=&Add_Vector(
     A=$Output.SkeletalMeshLocation001.SampledVelocity
     B=&Multiply_Vector(
      A=&Lerp_Vector(
       A=<1 -1.5 1>
       B=<1 1.5 1>
       Alpha=&MakeCustomFloatFromBool(Bool=$User.bIsFrontEnd, True Float=1, False Float=0)
      )
      B=&Lerp_Vector(
       A=&RandomRangeVector(Minimum=<-10 -40 -10>, Maximum=<10 -50 10>)
       B=&RandomRangeVector(Minimum=<0 -150 0>, Maximum=<1 -150 1>)
       Alpha=$System.Timeline.NormalizedTime
      )
     )
    )) {isSetParameters=true}
   /SetVariables(Particles.SpriteSize=&Multiply_Vector2(
     A=<2 2>
     B=&Vector2DFromFloat(Value=&UniformRangedFloat(Minimum=1, Maximum=2))
    )) {isSetParameters=true}
   &SubUVAnimation(
    End Frame Range Override=3
    Evaluation Type=::NewEnumerator0
    "Sprite Renderer "=DataInterface() {propertyValues {SpriteRenderer=&self:Drops_1.NiagaraSpriteRendererProperties_1}}
    SubUV Animation Mode:ENiagara_SubUVLookupModeV2=::NewEnumerator2
    UseEndFrame=true
    UseStartFrame=false
   )
   /SetVariables(Particles.Position=$Output.SkeletalMeshLocation001.SampledPosition) {isSetParameters=true}
   &CameraOffset(Camera Offset Amount=&RandomRangeFloat(Minimum=-40, Maximum=-20))
  ]
  particleUpdateScript [
   &ParticleState()
   &InheritSourceMovement(
    Applied Position Scale=&VectorFromFloat(Value=&FloatFromCurve(FloatCurve=DataInterface() {propertyValues {Curve {keys [
          {tangentMode=RCTM_User, time=0, value=1}
          {arriveTangent=-0.0405, interpMode=RCIM_Cubic, leaveTangent=-0.0405, tangentMode=RCTM_User, time=0.7009, value=1}
          {arriveTangent=-8.0089, interpMode=RCIM_Cubic, leaveTangent=-8.0089, tangentMode=RCTM_User, time=1, value=0}
         ]}}}))
    ApplyForce=false
    ApplyPosition=true
    ApplyPositionFalloff=false
   ) {enabled=false}
   &InheritSourceMovement(
    Applied Position Scale=&VectorFromFloat(Value=&FloatFromCurve(FloatCurve=DataInterface() {propertyValues {Curve {keys [
          {tangentMode=RCTM_User, time=0, value=1}
          {arriveTangent=-0.0405, interpMode=RCIM_Cubic, leaveTangent=-0.0405, tangentMode=RCTM_User, time=0.7009, value=1}
          {arriveTangent=-8.0089, interpMode=RCIM_Cubic, leaveTangent=-8.0089, tangentMode=RCTM_User, time=1, value=0}
         ]}}}))
    ApplyForce=false
    ApplyPosition=true
    ApplyPositionFalloff=false
   ) {enabled=false, module=InheritSourceMovement001}
   &GravityForce(Gravity=<0 0 -20>)
   &Drag(Drag=0, Use Linear Drag=true, Use Rotational Drag=false)
   &SolveForcesAndVelocity(Clamp Velocity=false, Limit Acceleration=false)
   &ScaleColor(
    Scale Alpha=&FloatFromCurve(
     FloatCurve=DataInterface() {propertyValues {Curve {keys [
         <at 0 0 RCIM_Cubic>
         {arriveTangent=1.6041, leaveTangent=1.6041, time=0.05, value=1.0113}
         {interpMode=RCIM_Cubic, leaveTangent=-1.0633, time=0.75, value=1}
         <at 0.99 0 RCIM_Cubic>
        ]}}}
     Scale Curve=0.3
    )
    ScaleA=true
    ScaleRGB=true
   )
   &ScaleSpriteSize(Scale Factor=&Vector2DFromFloat(Value=&FloatFromCurve(
      FloatCurve=DataInterface() {propertyValues {Curve {keys [
          {arriveTangent=-0.4919, interpMode=RCIM_Cubic, leaveTangent=-0.4919, tangentMode=RCTM_User, time=0.25, value=2}
          {arriveTangent=-3.3301, interpMode=RCIM_Cubic, leaveTangent=-3.3302, tangentMode=RCTM_User, time=1, value=0}
         ]}}}
      Scale Curve=0.65
     )))
  ]
  rendererClasses [&NiagaraSpriteRendererProperties]
  renderers [{rendererClass=&NiagaraSpriteRendererProperties}]
 }
 Empty {
  emitterUpdateScript [
   &EmitterState()
   &SpawnBurst_Instantaneous(Spawn Count=4, Use Spawn Probability=false, UseLoopCountLimit=false)
   &FN_SpawnAnimTrail(Number of Anim Trails=4)
  ]
  particleSpawnScript [
   &InitializeParticle(
    Color Mode=::NewEnumerator0
    Lifetime=400
    Sprite Size Mode=::NewEnumerator3
    Uniform Sprite Size=500
    UsePositionOffset=false
    Write Material Random=true
   ) {enabled=false}
   &FN_InitializeAnimTrailBoneIndecies(Second Bone or Socket Index=6)
   &FN_InitializeAnimTrailBoneIndecies(AnimTrail Selection=1, First Bone or Socket Index=6, Second Bone or Socket Index=2) {module=FN_InitializeAnimTrailBoneIndecies001}
   &FN_InitializeAnimTrailBoneIndecies(AnimTrail Selection=2, First Bone or Socket Index=3, Second Bone or Socket Index=7) {module=FN_InitializeAnimTrailBoneIndecies002}
   &FN_InitializeAnimTrailBoneIndecies(AnimTrail Selection=3, First Bone or Socket Index=7, Second Bone or Socket Index=5) {module=FN_InitializeAnimTrailBoneIndecies003}
   &FN_InitializeAnimTrail(
    AnimTrailWidth=1.5
    Color=<rgba 0 0 0 0.8>
    Lifetime=0.6
    Position Offset=&Lerp_Vector(
     A=<0 -15 0>
     B=<0 15 0>
     Alpha=&MakeCustomFloatFromBool(Bool=$User.bIsFrontEnd, True Float=1, False Float=0)
    )
   )
   &AddVelocity(Velocity=&Lerp_Vector(
     A=<0 1000 -111>
     B=<0 -1000 -111>
     Alpha=&MakeCustomFloatFromBool(Bool=$User.bIsFrontEnd, True Float=1, False Float=0)
    ))
   &ApplyInitialForces()
   /SetVariables() {isSetParameters=true, module=SetVariables_#3}
   &DynamicMaterialParameters(
    Index 1 Param 1=1
    Module.Param0WriteEnabled=true
    Module.Param0WriteEnabled1=true
    Module.Param1WriteEnabled=true
    Module.Param1WriteEnabled1=true
    Module.Param2WriteEnabled=true
    Module.Param2WriteEnabled1=true
    Module.Param3WriteEnabled=true
    Module.Param3WriteEnabled1=true
    Write Parameter Index 1=true
   )
   /SetVariables(Particles.Vector 2D:Vector2f=&Multiply_Vector2D_ByFloat(Float=1, Vector2D=<1 1>)) {isSetParameters=true}
  ]
  particleUpdateScript [
   &ParticleState()
   &FN_UpdateAnimTrail(Kill Locked particle on system deactivate=false, SkeletalMesh=$User.Skeletal Mesh)
   &InheritSourceMovement(ApplyForce=false, ApplyPosition=true, ApplyPositionFalloff=false) {enabled=false}
   &SolveForcesAndVelocity(Clamp Velocity=false, Limit Acceleration=false)
   &ScaleColor(
    Scale Alpha=&FloatFromCurve(FloatCurve=DataInterface() {propertyValues {Curve {keys [<at 0 0>,<at 1 1>]}}}, CurveIndex=$System.Timeline002.Time)
    ScaleA=true
    ScaleRGB=true
   ) {enabled=false, module=ScaleColor002}
   &ScaleColor(
    Scale Alpha=&FloatFromCurve(FloatCurve=DataInterface() {propertyValues {Curve {keys [<at 0 1>,<at 1 0>]}}})
    ScaleA=true
    ScaleRGB=true
   )
  ]
  rendererClasses [
   &NiagaraSpriteRendererProperties
   &NiagaraRibbonRendererProperties
   &NiagaraMeshRendererProperties
  ]
  renderers [
   {rendererClass=&NiagaraSpriteRendererProperties}
   {rendererClass=&NiagaraRibbonRendererProperties, rendererIndex=1}
   {rendererClass=&NiagaraRibbonRendererProperties, rendererIndex=2}
   {rendererClass=&NiagaraMeshRendererProperties, rendererIndex=3}
  ]
 }
 FN_Debug {
  bEnabled=false
  emitterUpdateScript [
   &EmitterState()
   &SpawnBurst_Instantaneous(Spawn Count=1, Use Spawn Probability=false)
  ]
  particleSpawnScript [/Niagara/Modules/Spawn/Initialization/InitializeParticle(
    Lifetime=5
    Module.Write Color=true
    Module.Write Lifetime=true
    Module.Write Mass=true
    Module.Write Material Random=true
    Module.Write Position=false
    Module.Write Scale=false
    Module.Write SpriteRotation=false
    Module.Write SpriteSize=true
    Position=&SimulationPosition()
    Sprite Size=<100 100>
   )]
  particleUpdateScript [
   &ParticleState()
   /SetVariables(Particles.Debug:NiagaraFloat=&Lerp_Float(A=0, B=1, Alpha=0)) {isSetParameters=true}
   &DynamicMaterialParameters(
    Index 0 Param 1=$Particles.Debug
    Module.Param0WriteEnabled=true
    Module.Param1WriteEnabled=true
    Module.Param2WriteEnabled=true
    Module.Param3WriteEnabled=true
   )
  ]
  rendererClasses [&NiagaraSpriteRendererProperties]
  renderers [{rendererClass=&NiagaraSpriteRendererProperties}]
 }
 SingleLoopingParticle {
  bEnabled=false
  emitterUpdateScript [
   &EmitterState()
   &SpawnBurst_Instantaneous(Spawn Count=10, Use Spawn Probability=false, UseLoopCountLimit=false)
  ]
  particleSpawnScript [/Niagara/Modules/Spawn/Initialization/V2/InitializeParticle(
    Color Mode=::NewEnumerator0
    Lifetime=$Emitter.CurrentLoopDuration
    Sprite Size Mode=::NewEnumerator3
    Uniform Sprite Size=20
    UsePositionOffset=false
    Write Material Random=true
   )]
  particleUpdateScript [
   &ParticleState()
   &SkeletalMeshLocation(
    Bone / Socket Index=&ReturnExecIndex()
    MaskSpawnGroup=false
    Mesh Sampling Type=::NewEnumerator5
    Skeletal Mesh=$User.SkeletalMesh
    Socket Sampling Mode=::NewEnumerator5
   )
   &SolveForcesAndVelocity(Clamp Velocity=false, Limit Acceleration=false)
  ]
  rendererClasses [&NiagaraSpriteRendererProperties]
  renderers [{rendererClass=&NiagaraSpriteRendererProperties}]
 }
 Smoke_Wave {
  emitterUpdateScript [
   &EmitterState(Life Cycle Mode=::NewEnumerator1, Loop Duration=1, Scalability Mode=::NewEnumerator1, UseLoopDelay=false)
   &self:ThresholdSpawnRate(MovementThreshold=50, SpawnRate=20) {enabled=false, module=ThresholdSpawnRate}
   &SpawnRate(SpawnRate=20, Use Spawn Probability=false)
  ]
  particleSpawnScript [
   &InitializeParticle(
    Color=<rgba 0 0 0 0.03>
    Lifetime Max=1
    Lifetime Min=0.9
    Lifetime Mode=::NewEnumerator1
    Sprite Rotation Mode=::NewEnumerator2
    Sprite Size Mode=::NewEnumerator1
    Uniform Sprite Size Max=45
    Uniform Sprite Size Min=35
    UsePositionOffset=false
    Write Material Random=true
   ) {module=InitializeParticle002}
   /SetVariables(Particles.Rotation:NiagaraFloat=&RandomRangeFloat(Minimum=-80, Maximum=80)) {isSetParameters=true}
   /SetVariables(Particles.RandomFrame=&RandomRangeInt(Minimum=0, Maximum=63)) {isSetParameters=true}
   &SphereLocation(
    MaskSpawnGroup=false
    Non Uniform Scale=<2 0 0>
    Override Seed=false
    Sphere Radius=10
    Surface Only=false
    UseNonUniformScale=true
    UseOffset=false
    UseSphereOrientation=false
   )
   &CameraOffset(Camera Offset Amount=&RandomRangeFloat(Minimum=5, Maximum=15))
   &AddVelocity(Velocity=&Lerp_Vector(
     A=<0 300 0>
     B=<0 -300 0>
     Alpha=&MakeCustomFloatFromBool(Bool=$User.bIsFrontEnd, True Float=1, False Float=0)
    ))
  ]
  particleUpdateScript [
   &ParticleState()
   &SkeletalMeshLocation(
    Bone / Socket Index=&RandomRangeInt(Minimum=0, Maximum=5)
    MaskSpawnGroup=false
    Mesh Sampling Type=::NewEnumerator5
    Sampled Position Offset=&Lerp_Vector(
     A=Linked() {linkedVariable {name=Particles.Initial.Position, type=NiagaraPosition}}
     B=&AddVectorToPosition(
      Position=$Particles.Initial.Position
      Vector=&Multiply_Vector(
       A=&RandomRangeVector(Minimum=<0 50 0>, Maximum=<0 150 1>)
       B=&VectorFromFloat(Value=&Lerp_Float(A=1, B=-1, Alpha=&MakeCustomFloatFromBool(Bool=$User.bIsFrontEnd, True Float=1, False Float=0)))
      )
     )
     Alpha=&FloatFromCurve(FloatCurve=DataInterface() {propertyValues {Curve {keys [{tangentMode=RCTM_User, time=0, value=0}, {tangentMode=RCTM_User, time=1, value=1}]}}})
    )
    Sampled Position Offset Coordinate Space:ENiagaraSkelMeshTransforms=::NewEnumerator2
    Skeletal Mesh=$User.SkeletalMesh
    Socket Sampling Mode=::NewEnumerator5
   )
   &Drag(Drag=2, Use Linear Drag=true, Use Rotational Drag=false)
   &InheritSourceMovement(Applied Position Scale=<-1 -1 -1>, ApplyForce=false, ApplyPosition=true, ApplyPositionFalloff=false) {enabled=false}
   &SolveForcesAndVelocity(Clamp Velocity=false, Limit Acceleration=false)
   &ScaleColor(
    Scale Alpha=&FloatFromCurve(
     FloatCurve=DataInterface() {propertyValues {Curve {keys [
         <at 0 0 RCIM_Cubic>
         {interpMode=RCIM_Cubic, tangentMode=RCTM_User, time=0.2, value=1}
         {arriveTangent=-1.25, interpMode=RCIM_Cubic, leaveTangent=-1.25, time=0.4, value=0.9932}
         <at 1 0>
        ]}}}
     Scale Curve=&FNTimeOfDayScalar(Night Scale=0.3, Morning Scale=0.1, Evening Scale=0.1)
    )
    ScaleA=true
    ScaleRGB=true
   ) {module=ScaleColor001}
   &ScaleSpriteSize(
    Uniform Curve Scale=&Lerp_Float(A=1.5, B=1, Alpha=&MakeCustomFloatFromBool(Bool=$User.bIsFrontEnd, True Float=1, False Float=0))
    Uniform Curve Sprite Scale=DataInterface() {propertyValues {Curve {keys [
        {arriveTangent=0.46, interpMode=RCIM_Cubic, leaveTangent=0.46, tangentMode=RCTM_User, time=0, value=1}
        {arriveTangent=0.2144, interpMode=RCIM_Cubic, leaveTangent=0.2144, tangentMode=RCTM_User, time=1, value=1.2}
       ]}}}
   ) {module=ScaleSpriteSize001}
   &FN_SubUVAnimation(End Frame=&Add_Integer(A=$Particles.RandomFrame, B=63), Start Frame=$Particles.RandomFrame, SubUV Loop Count=0.3)
  ]
  rendererClasses [&NiagaraSpriteRendererProperties, &NiagaraMeshRendererProperties]
  renderers [
   {rendererClass=&NiagaraSpriteRendererProperties}
   {rendererClass=&NiagaraMeshRendererProperties, rendererIndex=1}
  ]
 }
 WingsSplines {
  emitterUpdateScript [
   &EmitterState(Life Cycle Mode=::NewEnumerator1, Loop Behavior=::NewEnumerator1, Loop Duration Mode=::NewEnumerator1, UseLoopDelay=false)
   &SpawnBurst_Instantaneous(Spawn Count=16, Use Spawn Probability=false, UseLoopCountLimit=false)
  ]
  particleSpawnScript [
   &FN_SetSpawnGroup(Number of Spawn Groups=16)
   &InitializeParticle(
    Color Mode=::NewEnumerator0
    Lifetime=5
    Mesh Scale=&Multiply_Vector(
     A=&MakeVector(X=1, Y=0.5, Z=1.2)
     B=&SelectVectorFromArray(
      Direct Array Index=&ModuloInt(A=$Particles.SpawnGroup, B=5)
      Vector Selection Array=DataInterface() {propertyValues {FloatData [<0.3 0.6 0.7>, <0.5 0.5 0.7>, <0.5 0.5 0.5>, <0.5 0.5 0.5>, <0.5 0.5 0.6>]}}
     )
    )
    Mesh Scale Mode=::NewEnumerator4
    UsePositionOffset=false
    Write Material Random=true
   )
   &InitialMeshOrientation(
    Mesh Axis To Orient=<0 0 1>
    Mesh Orientation Mode=::NewEnumerator2
    Orientation Vector=<0 1 0>
    Rotation=&SelectVectorFromArray(
     Direct Array Index=$Particles.SpawnGroup
     Vector Selection Array=DataInterface() {propertyValues {FloatData [<0 -0.02 0.5>, <0 -0.01 0.5>, <0 -0.05 0.5>]}}
    )
    Use Rotation Vector=true
   ) {enabled=false}
   &CameraOffset()
  ]
  particleUpdateScript [
   &ParticleState(Kill Particles When Lifetime Has Elapsed=false, Let Infinitely Lived Particles Die When Emitter Deactivates=true)
   &SkeletalMeshLocation(
    Bone / Socket Index=&Add_Integer(A=4, B=$Particles.SpawnGroup)
    MaskSpawnGroup=false
    Mesh Sampling Type=::NewEnumerator5
    Sampled Position Offset=&SelectVectorFromArray(
     Direct Array Index=&ModuloInt(A=$Particles.SpawnGroup, B=8)
     Vector Selection Array=DataInterface() {propertyValues {FloatData [<0 0 -10>, <0 0 -20>, <0 0 -15>, <0 0 0>, <0 0 0>, <0 0 -10>, <0 0 -5>, <0 0 0>]}}
    )
    Skeletal Mesh=$User.SkeletalMeshWings
    Socket Sampling Mode=::NewEnumerator5
   )
   &InterpolateOverTime(
    Rate Of Change=10
    Target Vector Value:Vector3f=&ConvertPositionToVector(Input Position=$Particles.Position, Convert To Absolute Large World Vector=false)
    Variable Type=::NewEnumerator2
   )
   /SetVariables(Particles.Interpolate Dynamic=&MakeFloat4FromVectorAndFloat("Vector (RGB)"=$StackContext.InterpolateOverTime.InterpolateOverTime_V3_.Moving Average, "Float (Alpha)"=1.3)) {isSetParameters=true}
   &DynamicMaterialParameters(
    Index 0 Param 1=&RandomRangeFloat(Minimum=1.5, Maximum=1)
    Index 0 Param 2=&RandomRangeFloat(Minimum=1, Maximum=1.5)
    Index 0 Param 3=$User.SphereMaskRadius
    Index 0 Param 4=10
    Index 1 Param 1=1
    Index 2 Param 1=&Lerp_Float(
     A=4000
     B=&FloatFromCurve(FloatCurve=DataInterface() {propertyValues {Curve {keys [<at 0 80>,<at 0.3 10>]}}}, CurveIndex=$System.Timeline.Time, Scale Curve=0)
     Alpha=0
    )
    Module.Param0WriteEnabled=true
    Module.Param0WriteEnabled1=true
    Module.Param0WriteEnabled2=true
    Module.Param1WriteEnabled=true
    Module.Param1WriteEnabled1=true
    Module.Param1WriteEnabled2=true
    Module.Param2WriteEnabled=true
    Module.Param2WriteEnabled1=true
    Module.Param2WriteEnabled2=true
    Module.Param3WriteEnabled=true
    Module.Param3WriteEnabled1=true
    Module.Param3WriteEnabled2=true
    Write Parameter Index 1=true
    Write Parameter Index 2=true
   )
   /SetVariables(Particles.MaterialRandom:NiagaraFloat=&RandomRangeFloat(Minimum=0, Maximum=1)) {isSetParameters=true}
   &ScaleColor(
    Scale Alpha=&FN_FrontEndFloatModifier(
     Gameplay=&Multiply_Float(
      A=$System.DiveOpacity
      B=&FNTimeOfDayScalar(Night Scale=4, Morning Scale=2.5, Day Scale=2.5, Evening Scale=2)
     )
     Front End=1.5
    )
    ScaleA=true
    ScaleRGB=true
   ) {enabled=false}
   &ScaleMeshSize(Scale Factor=<5 5 2.5>)
   &ScaleMeshSize(Scale Factor=&VectorFromFloat(Value=&SelectFloatFromArray(
      Direct Array Index=&ModuloInt(A=$Particles.SpawnGroup, B=8)
      Float Selection Array=DataInterface() {propertyValues {FloatData [1,0.9,0.9,0.7,0.63,0.6,0.5,0.5]}}
     ))) {module=ScaleMeshSize001}
   &ScaleColor(
    Scale Alpha=&FloatFromCurve(
     FloatCurve=DataInterface() {propertyValues {Curve {keys [{tangentMode=RCTM_User, time=0, value=1}, {tangentMode=RCTM_User, time=1, value=0}]}}}
     CurveIndex=$System.Timeline002.Time
    )
    ScaleA=true
    ScaleRGB=true
   ) {module=ScaleColor002}
   &ScaleColor(Scale Alpha=0.4, ScaleA=true, ScaleRGB=true) {module=ScaleColor001}
  ]
  rendererClasses [&NiagaraMeshRendererProperties]
  renderers [
   {rendererClass=&NiagaraMeshRendererProperties, rendererIndex=0}
   {rendererClass=&NiagaraMeshRendererProperties, rendererIndex=1}
   {rendererClass=&NiagaraMeshRendererProperties, rendererIndex=2}
   {rendererClass=&NiagaraMeshRendererProperties, rendererIndex=3}
  ]
 }
 WingsTip {
  emitterUpdateScript [
   &EmitterState()
   &FN_SpawnAnimTrail(AnimTrailWidthMode=::FromFirst, Number of Anim Trails=2)
  ]
  particleSpawnScript [
   &InitializeParticle(Color Mode=::NewEnumerator0, Lifetime=0.1, UsePositionOffset=false, Write Material Random=true) {enabled=false}
   &FN_InitializeAnimTrailBoneIndecies()
   &FN_InitializeAnimTrailBoneIndecies(AnimTrail Selection=1, First Bone or Socket Index=3, Second Bone or Socket Index=4) {module=FN_InitializeAnimTrailBoneIndecies001}
   &FN_InitializeAnimTrail(
    AnimTrailWidth=0.5
    Color=<rgba 0 0 0 1>
    Lifetime=0.4
    Position Offset=&MaskVectorBySpawnGroup(Masked Vector=<-13 -11 7>, Passthrough Vector=<0 -10 3>)
   )
   &AddVelocity(Velocity=<0 1000 -111>)
   &ApplyInitialForces()
   /SetVariables() {isSetParameters=true, module=SetVariables_#3}
  ]
  particleUpdateScript [
   &ParticleState()
   &FN_UpdateAnimTrail(Kill Locked particle on system deactivate=false, SkeletalMesh=$User.Skeletal Mesh)
   &InheritSourceMovement(ApplyForce=false, ApplyPosition=true, ApplyPositionFalloff=false) {enabled=false}
   &SolveForcesAndVelocity(Clamp Velocity=false, Limit Acceleration=false)
   &ScaleColor(
    Scale Alpha=&FloatFromCurve(FloatCurve=DataInterface() {propertyValues {Curve {keys [<at 0 0>,<at 1 1>]}}}, CurveIndex=$System.Timeline002.Time)
    ScaleA=true
    ScaleRGB=true
   ) {module=ScaleColor002}
  ]
  rendererClasses [&NiagaraSpriteRendererProperties, &NiagaraRibbonRendererProperties]
  renderers [
   {rendererClass=&NiagaraSpriteRendererProperties}
   {rendererClass=&NiagaraRibbonRendererProperties, rendererIndex=1}
   {rendererClass=&NiagaraRibbonRendererProperties, rendererIndex=2}
  ]
 }
 Wings_Smudge {
  emitterUpdateScript [
   &EmitterState(
    Inactive Response=::NewEnumerator1
    Life Cycle Mode=::NewEnumerator1
    Loop Behavior=::NewEnumerator1
    Loop Duration Mode=::NewEnumerator1
    UseLoopDelay=true
   )
   &SpawnBurst_Instantaneous(Spawn Count=16, Use Spawn Probability=false, UseLoopCountLimit=false)
   &SpawnBurst_Instantaneous(
    Spawn Count=&MakeCustomIntFromBool(
     Bool=&IsPlatformSetActive(New Platform Set=DataInterface() {propertyValues {Platforms {qualityLevelMask=-4}}})
     True INT=4
     False INT=0
    )
    Spawn Group=1
    Use Spawn Probability=false
    UseLoopCountLimit=false
   ) {enabled=false, module=SpawnBurst_Instantaneous001}
  ]
  particleSpawnScript [
   &InitializeParticle(
    Color Mode=::NewEnumerator0
    Lifetime=9999999
    Mesh Scale Mode=::NewEnumerator3
    Mesh Uniform Scale=&MakeCustomFloatFromBool(
     Bool=&IsPlatformSetActive(New Platform Set=DataInterface() {propertyValues {Platforms {qualityLevelMask=-4}}})
     True Float=&Lerp_Float(
      A=0.2
      B=0.5
      Alpha=&FloatFromCurve(
       FloatCurve=DataInterface() {propertyValues {Curve {keys [
           {arriveTangent=2.2061, interpMode=RCIM_Cubic, leaveTangent=2.2061, tangentMode=RCTM_User, time=0, value=0}
           {arriveTangent=0.0224, interpMode=RCIM_Cubic, leaveTangent=0.0224, tangentMode=RCTM_User, time=1, value=1}
          ]}}}
       CurveIndex=&RandomRangeFloat(Minimum=0, Maximum=1)
      )
     )
     False Float=&Lerp_Float(
      A=0.35
      B=0.5
      Alpha=&FloatFromCurve(
       FloatCurve=DataInterface() {propertyValues {Curve {keys [
           {arriveTangent=2.2061, interpMode=RCIM_Cubic, leaveTangent=2.2061, tangentMode=RCTM_User, time=0, value=0}
           {arriveTangent=0.0224, interpMode=RCIM_Cubic, leaveTangent=0.0224, tangentMode=RCTM_User, time=1, value=1}
          ]}}}
       CurveIndex=&RandomRangeFloat(Minimum=0, Maximum=1)
      )
     )
    )
    Position Mode=::NewEnumerator0
    Sprite Size Mode=::NewEnumerator3
    Uniform Sprite Size=30
    Write Material Random=true
   )
   &PartitionParticles(
    ID to Partition:NiagaraInt32=&MaskIntBySpawnGroup(
     Masked Int=&self:ReturnUniqueID()
     Passthrough Int=&RandomRangeInt(Minimum=0, Maximum=8)
     Spawn Group Mask=0
    )
    Number Of Partitions=16
   )
  ]
  particleUpdateScript [
   &ParticleState()
   &SkeletalMeshLocation(
    Bone / Socket Index=$Particles.PartitionParticles.Partition
    MaskSpawnGroup=false
    Mesh Sampling Type=::NewEnumerator5
    Orientation Sampling=::NewEnumerator1
    Position Sampling=::NewEnumerator2
    Sampled Orientation Coordinate Space=::Local
    Sampled Position Coordinate Space=::World
    Sampled Position Offset=<0 -10 0>
    Sampled Position Offset Coordinate Space:ENiagaraSkelMeshTransforms=::NewEnumerator2
    Skeletal Mesh=$User.SkeletalMeshLeft
    Socket Sampling Mode=::NewEnumerator5
   )
   &SkeletalMeshLocation(
    MaskSpawnGroup=false
    Mesh Sampling Type=::NewEnumerator5
    Orientation Sampling=::NewEnumerator1
    Position Sampling=::NewEnumerator2
    Sampled Orientation Coordinate Space=::Local
    Sampled Position Coordinate Space=::World
    Sampled Position Offset Coordinate Space:ENiagaraSkelMeshTransforms=::NewEnumerator2
    Skeletal Mesh=$User.Skeletal Mesh2
    Socket Sampling Mode=::NewEnumerator5
   ) {module=SkeletalMeshLocation001}
   /SetVariables(
    Particles.ArmDistance:NiagaraFloat=&Multiply_Float(A=-150, B=&MakeFloatFromVector(Vector=$Particles.Scale, Channel=::NewEnumerator0))
    Particles.InterpolateSpeed:NiagaraFloat=&Lerp_Float(A=15, B=10, Alpha=$System.Timeline.NormalizedTime)
    Particles.MeshOrientationRoot:Quat4f=&OrientMeshToVector(
     Look At Direction=&TransformVector(
      Vector=&RotateVectorByQuaternion(Vector To Rotate=<0 0 0>, Quaternion=$Particles.SkeletalMeshLocation.SampledRotation)
      Destination Space=::World
     )
     Reference Axis Vector=<1 0 0>
    )
    Particles.MeshVectorRoot:Vector3f=&RotateVectorByQuaternion(Vector To Rotate=<0 0 0>, Quaternion=$Particles.SkeletalMeshLocation001.SampledRotation)
    Particles.MeshVectorWing:Vector3f=&RotateVectorByQuaternion(Vector To Rotate=<0 0 0>, Quaternion=$Particles.SkeletalMeshLocation001.SampledRotation)
    Particles.RestAxisLocalSpace:Vector3f=<0 1 0>
    Particles.RootPosition:NiagaraPosition=$Particles.SkeletalMeshLocation.SampledPosition
    Particles.WingPos:Vector3f=$Particles.SkeletalMeshLocation001.SampledPosition
   ) {isSetParameters=true}
   &InterpolateOverTime(Rate Of Change=2, Target Float Value:NiagaraFloat=$User.RightDot) {module=InterpolateOverTime001}
   /SetVariables(Particles.InterpolatePos:Vector3f=&Add_Vector(
     A=&TransformVector(
      Vector=&Add_Vector(
       A=&Multiply_VectorByFloat(Vector=$Particles.MeshVectorWing, Float=100)
       B=&Add_Vector(
        A=&Lerp_Vector(A=<0 -200 80>, B=<0 -400 270>, Alpha=$System.Timeline.NormalizedTime)
        B=&Lerp_Vector(A=<0 0 0>, B=<400 1 1>, Alpha=$Particles.InterpolateOverTime001.InterpolateOverTime_Float.Moving Average)
       )
      )
      Destination Space=::World
     )
     B=$Particles.WingPos
    )) {isSetParameters=true}
   /SetVariables(Particles.MeshOrientation=&FindQuatBetweenTwoVectors(
     From Vector=<1 0 0>
     To Vector=&TransformVector(Vector=$Particles.RestAxisLocalSpace, Destination Space=::World)
    )) {isSetParameters=true}
   &InterpolateOverTime(
    Initial Vector Value:Vector3f=$Particles.RootPosition
    Rate Of Change:NiagaraFloat=$Particles.InterpolateSpeed
    Target Vector Value:Vector3f=$Particles.InterpolatePos
    Variable Type=::NewEnumerator2
   )
   &self:PositionTest(
    ArmLength:NiagaraFloat=$Particles.ArmDistance
    In:Vector3f=$Particles.InterpolateOverTime.InterpolateOverTime_V3_.Moving Average
    InterpPos:Vector3f=$Particles.InterpolatePos
    Max Angle=80
    RestAxis:Vector3f=$Particles.RestAxisLocalSpace
   ) {module=PositionTest}
   /SetVariables(Particles.OffsetAge:NiagaraFloat=&Subtract_Float(
     A=&Add_Float(A=$Particles.Age, B=&Multiply_FloatByInt(Float=100, Integer=$User.Deployed))
     B=0.5
    )) {isSetParameters=true}
   /SetVariables(Particles.Position_Constraint:Vector3f=&Add_Vector(
     A=$Particles.RootPosition
     B=&Multiply_VectorByFloat(
      Vector=&NormalizeVector(
       Vector To Normalize=&Subtract_Vector(A=$Particles.PositionTest.Clamped, B=$Particles.RootPosition)
       Safe Normalize=true
       Safe Fallback Threshold=0
       Safe Fallback Vector=<0 0 0.001>
      )
      Float=$Particles.ArmDistance
     )
    )) {isSetParameters=true}
   &ScaleMeshSize(Scale Factor=&MakeVector(
     X=&FloatFromCurve(
      FloatCurve=DataInterface() {propertyValues {Curve {keys [{interpMode=RCIM_Cubic, tangentMode=RCTM_User, time=0, value=0}, {interpMode=RCIM_Cubic, tangentMode=RCTM_User, time=1, value=1}]}}}
      CurveIndex=$Particles.OffsetAge
     )
     Y=&FloatFromCurve(
      FloatCurve=DataInterface() {propertyValues {Curve {keys [{tangentMode=RCTM_User, time=0.2, value=1}, {tangentMode=RCTM_User, time=1, value=1}]}}}
      CurveIndex=$Particles.OffsetAge
     )
     Z=1
    )) {module=ScaleMeshSize001}
   &FN_Timer(Increment Multiplier:NiagaraFloat=&Lerp_Float(A=0.7, B=1.2, Alpha=$System.Timeline.Time))
   &DynamicMaterialParameters(
    Index 0 Param 1=&MakeFloatFromVector(Vector=$Particles.Position_Constraint, Channel=::NewEnumerator0)
    Index 0 Param 2=&MakeFloatFromVector(Vector=$Particles.Position_Constraint, Channel=::NewEnumerator1)
    Index 0 Param 3=&MakeFloatFromVector(Vector=$Particles.Position_Constraint, Channel=::NewEnumerator2)
    Index 0 Param 4=&MakeFloatFromVector(
     Vector=&TransformVector(Vector=$Particles.RestAxisLocalSpace, Destination Space=::World)
     Channel=::NewEnumerator0
    )
    Index 1 Param 1=&MakeFloatFromVector(
     Vector=&TransformVector(Vector=$Particles.RestAxisLocalSpace, Destination Space=::World)
     Channel=::NewEnumerator1
    )
    Index 1 Param 2=&MakeFloatFromVector(
     Vector=&TransformVector(Vector=$Particles.RestAxisLocalSpace, Destination Space=::World)
     Channel=::NewEnumerator2
    )
    Index 1 Param 3=&MakeFloatFromVector(Vector=$Particles.RootPosition, Channel=::NewEnumerator0)
    Index 1 Param 4=&MakeFloatFromVector(Vector=$Particles.RootPosition, Channel=::NewEnumerator1)
    Index 2 Param 1=&MakeFloatFromVector(Vector=$Particles.RootPosition, Channel=::NewEnumerator2)
    Index 2 Param 2=1
    Index 2 Param 4=1
    Index 3 Param 1=&Add_Float(
     A=$Particles.FN_Timer.Value
     B=/Game/Effects/Niagara/DynamicInputs/Random/FN_SeededRandomRangeFloat(
      Minimum:NiagaraFloat=0
      Maximum:NiagaraFloat=50
      Random Seed:NiagaraInt32=&self:ReturnUniqueID_03()
     )
    )
    Index 3 Param 2=&MakeFloatFromVector(Vector=$Particles.Scale, Channel=::NewEnumerator0)
    Index 3 Param 3=&Lerp_Float(A=0.036, B=0, Alpha=$System.Timeline.NormalizedTime)
    Index 3 Param 4=&Lerp_Float(A=0.782, B=0, Alpha=$System.Timeline.NormalizedTime)
    Module.Param0WriteEnabled=true
    Module.Param0WriteEnabled1=true
    Module.Param0WriteEnabled2=true
    Module.Param0WriteEnabled3=true
    Module.Param1WriteEnabled=true
    Module.Param1WriteEnabled1=true
    Module.Param1WriteEnabled2=true
    Module.Param1WriteEnabled3=true
    Module.Param2WriteEnabled=true
    Module.Param2WriteEnabled1=true
    Module.Param2WriteEnabled2=true
    Module.Param2WriteEnabled3=true
    Module.Param3WriteEnabled=true
    Module.Param3WriteEnabled1=true
    Module.Param3WriteEnabled2=true
    Module.Param3WriteEnabled3=true
    Write Parameter Index 1=true
    Write Parameter Index 2=true
    Write Parameter Index 3=true
   )
   /SetVariables(Particles.Position=&Add_Vector(
     A=$Particles.RootPosition
     B=&TransformVector(
      Vector=&MakeVector(
       X=0
       Y=&FloatFromCurve(
        FloatCurve=DataInterface() {propertyValues {Curve {keys [
            {arriveTangent=-0.2987, interpMode=RCIM_Cubic, leaveTangent=-0.2987, tangentMode=RCTM_User, time=0, value=0.75}
            {arriveTangent=-0.0552, interpMode=RCIM_Cubic, leaveTangent=-0.0552, tangentMode=RCTM_User, time=1, value=0}
           ]}}}
        CurveIndex=$Particles.OffsetAge
        Scale Curve=15
       )
       Z=0
      )
      Destination Space=::World
     )
    )) {isSetParameters=true}
   /SetVariables(Particles.Testend:Vector3f=$Particles.InterpolatePos) {enabled=false, isSetParameters=true}
   &Color(Scale Color=<0 1 1>) {enabled=false}
   &ScaleColor(
    Scale Alpha=&MakeCustomFloatFromBool(
     Bool=$System.DefaultVarient
     True Float=5
     False Float=&FNTimeOfDayScalar(Night Scale=0.5, Morning Scale=1, Evening Scale=1)
    )
    Scale RGB=<0 0 0>
    ScaleA=true
    ScaleRGB=true
   )
   /SetVariables(Particles.StrandsColor:LinearColor=<rgba 0 0 0 0.4>) {isSetParameters=true}
   &ScaleColor(
    Scale Alpha=&FloatFromCurve(FloatCurve=DataInterface() {propertyValues {Curve {keys [<at 0 0>,<at 1 1>]}}}, CurveIndex=$System.Timeline002.Time)
    ScaleA=true
    ScaleRGB=true
   ) {module=ScaleColor002}
   &ScaleColor(
    Scale Alpha=&FloatFromCurve(FloatCurve=DataInterface() {propertyValues {Curve {keys [<at 0 1>,<at 1 0>]}}}, CurveIndex=$System.Timeline001.Time)
    ScaleA=true
    ScaleRGB=true
   ) {enabled=false, module=ScaleColor001}
   &ScaleColor(Scale Alpha=0.5, ScaleA=true, ScaleRGB=true) {module=ScaleColor003}
  ]
  rendererClasses [&NiagaraMeshRendererProperties, &NiagaraSpriteRendererProperties]
  renderers [
   {rendererClass=&NiagaraMeshRendererProperties, rendererIndex=0}
   {rendererClass=&NiagaraSpriteRendererProperties, rendererIndex=1}
   {rendererClass=&NiagaraMeshRendererProperties, rendererIndex=2}
   {rendererClass=&NiagaraMeshRendererProperties, rendererIndex=3}
   {rendererClass=&NiagaraMeshRendererProperties, rendererIndex=4}
  ]
 }
}
summary {
 systemName=NS_Glider_AgentSherbert_02
 emitters [
  {emitterName=Wings_Smudge, rendererClasses [&NiagaraMeshRendererProperties, &NiagaraSpriteRendererProperties]}
  {emitterName=Drops, simTarget=GPUComputeSim, rendererClasses [&NiagaraSpriteRendererProperties]}
  {emitterName=WingsSplines, rendererClasses [&NiagaraMeshRendererProperties]}
  {bEnabled=false, emitterName=FN_Debug, rendererClasses [&NiagaraSpriteRendererProperties]}
  {
   emitterName=Empty
   rendererClasses [
    &NiagaraSpriteRendererProperties
    &NiagaraRibbonRendererProperties
    &NiagaraMeshRendererProperties
   ]
  }
  {emitterName=WingsTip, rendererClasses [&NiagaraSpriteRendererProperties, &NiagaraRibbonRendererProperties]}
  {emitterName=Smoke_Wave, rendererClasses [&NiagaraSpriteRendererProperties, &NiagaraMeshRendererProperties]}
  {bEnabled=false, emitterName=SingleLoopingParticle, rendererClasses [&NiagaraSpriteRendererProperties]}
 ]
 userVariables [
  User.bIsFrontEnd {type=NiagaraBool, defaultValue=true}
  User.bIsFrontEndPreview {type=NiagaraBool}
  User.bisFullyDeployed {type=NiagaraBool}
  User.Delay {type=NiagaraFloat, defaultValue=0, description=""}
  User.Deployed {type=NiagaraInt32}
  User.ForwardDot {type=NiagaraFloat, defaultValue=0, description=""}
  User.IsFrontEnd {type=NiagaraBool}
  User.isVarient {type=NiagaraFloat, defaultValue=0, description=""}
  User.RibbonIDOffset {type=NiagaraInt32}
  User.RightDot {type=NiagaraFloat, defaultValue=0, description=""}
  User.Skeletal Mesh {
   type=NiagaraDataInterfaceSkeletalMesh
   defaultValue {
    struct=NiagaraExt_VariableValue_DataInterface
    value {dataInterface=&self:NiagaraDataInterfaceSkeletalMesh_4}
   }
  }
  User.Skeletal Mesh2 {
   type=NiagaraDataInterfaceSkeletalMesh
   defaultValue {
    struct=NiagaraExt_VariableValue_DataInterface
    value {dataInterface=&self:NiagaraDataInterfaceSkeletalMesh_0}
   }
  }
  User.SkeletalMesh {
   type=NiagaraDataInterfaceSkeletalMesh
   defaultValue {
    struct=NiagaraExt_VariableValue_DataInterface
    value {dataInterface=&self:NiagaraDataInterfaceSkeletalMesh_5}
   }
  }
  User.SkeletalMeshLeft {
   type=NiagaraDataInterfaceSkeletalMesh
   defaultValue {
    struct=NiagaraExt_VariableValue_DataInterface
    value {dataInterface=&self:NiagaraDataInterfaceSkeletalMesh_2}
   }
  }
  User.SkeletalMeshWings {
   type=NiagaraDataInterfaceSkeletalMesh
   defaultValue {
    struct=NiagaraExt_VariableValue_DataInterface
    value {dataInterface=&self:NiagaraDataInterfaceSkeletalMesh_3}
   }
  }
  User.SphereMaskRadius {type=NiagaraFloat, defaultValue=0, description=""}
  User.Variant {type=NiagaraFloat, defaultValue=0, description=""}
 ]
}
